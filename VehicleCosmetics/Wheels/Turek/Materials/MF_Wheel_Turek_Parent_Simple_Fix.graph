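# Material function for the Turek wheel that builds on MF_WheelParent_Simple and adds an animated noise effect to its emissive.
# The green channel of a scaled, panning noise texture is shaped by power and scale, with a small lift for dark areas.
# Multiplied by the tire mask's red channel and saturated, it modulates emissive color.
# A final attribute blend, noted as necessary for wheels with tires, produces the output.
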MaterialFunction /VehicleCosmetics/Wheels/Turek/Materials/MF_Wheel_Turek_Parent_Simple_Fix
graph {nodes [
  n1 = GetMaterialAttributes(None=NamedRerouteUsage()) {attributeGetTypes [#1]}
  n2 = TextureCoordinate()
  n3 = VectorParameter() {
   group=5_Special Effects
   parameterName=FX Parameters
   sortPriority=-3
   channelNames {a=LUM Scale, b=LUM Power, g=Speed, r=Scale}
   defaultValue=<rgba 1 0.1728 3 6>
  }
  n4 = Time()
  n5 = TextureSampleParameter2D(UVs=Panner(Coordinate=Multiply(A=@n2, B=@n3.Scale), Time=Multiply(A=@n4, B=@n3.Speed)) {speedX=0.973, speedY=1.228}) {group=5_Special Effects, parameterName=FX_Texture, sortPriority=-5, texture=/VehicleCosmetics/Wheels/_SharedTextures/FX/RGBNoise_01}
  n6 = ComponentMask(None=@n5.RGB) {g=true}
  n7 = Add(A=Multiply(A=Power(Base=@n6, Exp=@n3.LUM Power), B=@n3.LUM Scale), B=Constant() {r=0.02}) {desc=Lighten the darkest emissives.}
  n8 = ComponentMask(None=NamedRerouteUsage()) {desc=R Channel of Tire Mask, r=true}
  n9 = Multiply(A=@n1.EmissiveColor, B=Saturate(None=Multiply(A=@n7, B=@n8)))
  FunctionOutput(None=BlendMaterialAttributes(
    A=NamedRerouteUsage()
    B=SetMaterialAttributes!(MaterialAttributes=@n1.MaterialAttributes, EmissiveColor=@n9) {attributeSetTypes [#1]}
    Alpha=NamedRerouteUsage()
   ) {desc=Necessary if Wheel Has Tire}) {outputName=MaterialAttributes}
  n10 = MaterialFunctionCall() {materialFunction=&MF_WheelParent_Simple}
  NamedRerouteDeclaration(None=@n10.MaterialAttributes) {name=MF_WheelParent_Simple_Out, nodeColor=<rgba 0 1 0.6484 1>}
  NamedRerouteDeclaration(None=@n10.TireRGBAMask) {name=TireRGBA_Out, nodeColor=<rgba 0.2656 0 1 1>}
  NamedRerouteDeclaration(None=@n10.VertexColorR) {name=VertColor_R_Out, nodeColor=<rgba 1 0.5 0 1>}
  Reroute()
  Reroute()
  Reroute()
  Reroute()
  Reroute()
  Reroute()
  Reroute()
  Reroute()
  Reroute()
  Reroute()
  Reroute()
  Reroute()
  Reroute()
  Reroute()
  Reroute()
  Reroute()
  Reroute()
  Reroute()
  Reroute()
  Reroute()
  Reroute()
  Reroute()
  Reroute()
  Reroute()
 ]}
